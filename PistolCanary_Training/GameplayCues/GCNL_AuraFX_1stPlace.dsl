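; Looping gameplay cue for the first-place aura effect, named for PistolCanary Training, applied to a Fort player pawn.
; On application it caches the pawn and binds a spawned Niagara component's "SK" skeletal mesh parameter to the pawn's Body part mesh.
; For a locally viewed pawn it also pushes a sound mix modifier, which removal pops under the same check.
; Debug prints label that mix GrillMoon_SafeZone_MixMod.
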(event EventOnApplication(EitherVFX-preferthisone)
  (params (MyTarget :type Actor%20Object%20Reference) (Parameters :type Gameplay%20Cue%20Parameters%20Structure%20%28by%20ref%29) (ParticleComponents :type Array%20of%20FXSystem%20Component%20Object%20References) (AudioComponents :type Array%20of%20Audio%20Component%20Object%20References) (BurstCameraShakeInstance :type Camera%20Shake%20Base%20Object%20Reference) (BurstDecalInstance :type Decal%20Actor%20Object%20Reference))
  (bind _asfort_player_pawn (Utilities|Casting|CastToFortPlayerPawn MyTarget))
  (Variables|Default|SetFort%20Player%20Pawn _asfort_player_pawn)
  (Utilities|FlowControl|Sequence
    (:then_0
      (bind _asniagara_particle_system_component (Utilities|Casting|CastToNiagaraComponent (Utilities|Array|Get(acopy) ParticleComponents)))
      (Niagara|SetNiagaraSkeletalMeshComponent _asniagara_particle_system_component "SK" (Customization|GetSkeletalMeshforPartType (Variables|Default|GetFort%20Player%20Pawn) "Body" :!class /Script/FortniteGame.FortPlayerPawn)))
    (:then_1
      (Utilities|FlowControl|Sequence
        (:then_0
          (if (Pawn|IsLocallyViewed (Variables|Default|GetFort%20Player%20Pawn))
            (Audio|PushSoundMixModifier 0 (Variables|Default|GetSound%20Class%20Mix))
            (Development|PrintString 0 "Activated GrillMoon_SafeZone_MixMod")))))))

(event EventOnRemoval(EitherVFX-preferthisversion)
  (params (MyTarget :type Actor%20Object%20Reference) (Parameters :type Gameplay%20Cue%20Parameters%20Structure%20%28by%20ref%29) (ParticleComponents :type Array%20of%20FXSystem%20Component%20Object%20References) (AudioComponents :type Array%20of%20Audio%20Component%20Object%20References) (BurstCameraShakeInstance :type Camera%20Shake%20Base%20Object%20Reference) (BurstDecalInstance :type Decal%20Actor%20Object%20Reference))
  (bind _fort_player_pawn (Variables|Default|GetFort%20Player%20Pawn)
    (:then
      (if (Pawn|IsLocallyViewed _fort_player_pawn)
        (Audio|PopSoundMixModifier 0 (Variables|Default|GetSound%20Class%20Mix))
        (Development|PrintString 0 "Deactivated GrillMoon_SafeZone_MixMod")))
    (:else)))

(fn ConstructionScript)
